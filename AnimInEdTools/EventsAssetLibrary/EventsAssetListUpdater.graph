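# Editor utility widget that fills a per-project Events Asset Library DataTable from the focused Level Sequence.
# The Add button saves the sequence, recursively walks subsequences, and turns binding templates into rows for checkbox-enabled mannequin, skeletal mesh, static mesh and other actor types.
# It finds the table by project name, creating it through embedded Python if missing.
# UpdateDataTable writes only Fort mannequin rows, keyed by display name, reporting added or updated counts.
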EditorUtilityWidgetBlueprint /AnimInEdTools/EventsAssetLibrary/EventsAssetListUpdater
graphs [
 CreateNewDataTable {nodes [
   CreateNewDataTable() {is=title K2Node_FunctionEntry}
   n1 = Get Current Project() -> (Current Project:BP Animation Browser Project Object Reference)
   > Execute Python Script!(
    execute
    PythonScript="import unreal\r\n\r\nasset_name = name\r\npackage_path = path\r\n\r\nrow_struct = unreal.load_asset(\"&/Struct_EventsAssetPicker\")\r\n\r\nfactory = unreal.DataTableFactory()\r\nfactory.set_editor_property(\"struct\", row_struct)\r\n\r\nasset_tools = unreal.AssetToolsHelpers.get_asset_tools()\r\n\r\ndata_table = asset_tools.create_asset(\r\n    asset_name=asset_name,\r\n    package_path=package_path,\r\n    asset_class=unreal.DataTable,\r\n    factory=factory\r\n)\r\n\r\nunreal.EditorAssetLibrary.save_loaded_asset(data_table)"
    name:String=Append(A=DT_EAL_, B=Get Name!(self:BP Animation Browser Project Object Reference=@n1) -> (Name:String))
    path:String=MakeLiteralString(Value=&/DataTables/)
   ) -> (then, ReturnValue) {inputs [name,path]}
   Convert Struct To Json String!(execute:Exec, Struct:"Wildcard (by ref)") -> (then:Exec, OutJsonString:String, ReturnValue:bool) {is=title K2Node_StructToJsonString}
  ]}
 ED_ListViewItemSelected {nodes [ED_ListViewItemSelected!() -> (then, ItemSelected:bool=~) {is=title K2Node_FunctionEntry}]}
 EventGraph {nodes [
   "On Clicked (AddAssetsBtn)"() {is=title K2Node_ComponentBoundEvent}
   n1 = Get Project_ComboBox() -> (Project_ComboBox:"Searchable Combo Box (String) Object Reference")
   > n2 = Branch(Condition="Equal Exactly (String)"(A=GetSelectedOption!(self:"Searchable Combo Box (String) Object Reference"=@n1) -> (ReturnValue:String)))
   > SetWarning!(execute, self, Text:Text="LOCTEXT(\"Select Project\")") -> (then)
   Set ItemsAdded(execute=@n2.else, ItemsAdded:int=0) -> (Output_Get:int=0)
   > Clear(TargetArray:Array of Struct Events Asset Picker Structures=Get TableRowStruct_AssetsToAdd() -> (TableRowStruct_AssetsToAdd:Array of Struct Events Asset Picker Structures))
   > n3 = GetFocusedLevelSequence!(execute) -> (then, ReturnValue:Level Sequence Object Reference)
   > Is Valid(InputObject=@n3)
   > n4 = Set Level Sequence(Level Sequence:Level Sequence Object Reference=@n3) -> (Output_Get:Level Sequence Object Reference)
   > SaveLoadedAsset(AssetToSave=@n4, bOnlyIfIsDirty=true)
   > FindSubSequence!(execute, self, Sequence:Movie Scene Sequence Object Reference=@n4) -> (then)
   n5 = "On Opening (Project_ComboBox)"() {is=title K2Node_ComponentBoundEvent}
   UpdateComboBox!(
    execute=[@n5.then, @n8.then]
    self
    Combo Box:"Searchable Combo Box (String) Object Reference"=Get Project_ComboBox()
    Items:Array of Strings=Get ProjectNames() -> (ProjectNames:Array of Strings)
    DefaultItem:String
    AddBlank:bool=true
   ) -> (then)
   n6 = "On Selection Changed (Project_ComboBox)"!() -> (OutputDelegate, then, SelectedItem:String, SelectionType:ESelectInfo Enum) {is=title K2Node_ComponentBoundEvent}
   > Branch(Condition=OR Boolean!(
     A="Equal (Enum)"(A:ESelectInfo Enum=@n6.SelectionType, B:ESelectInfo Enum=OnKeyPress)
     B="Equal (Enum)"(A=@n6.SelectionType, B=OnMouseClick)
     C:bool="Equal (Enum)"(A=@n6.SelectionType, B=OnKeyPress)
    ) -> (ReturnValue))
   n7 = Get ProjectNames()
   > Set Current Project(Current Project:BP Animation Browser Project Object Reference="Get (a copy)"(
     Array:Array of BP Animation Browser Project Object References=Get CurrentProjects() -> (CurrentProjects:Array of BP Animation Browser Project Object References)
     Dimension 1=Find Item!(TargetArray=@n7, ItemToFind:"String (by ref)"=@n6.SelectedItem) -> (ReturnValue)
    ) -> (Output:BP Animation Browser Project Object Reference)) -> (Output_Get:BP Animation Browser Project Object Reference)
   > UpdateComboBox'()
   n8 = UpdateComboBox'() {is=title K2Node_CustomEvent}
   Event Construct()
   > n9 = Set CurrentProjects(CurrentProjects:Array of BP Animation Browser Project Object References=GetProjects!() -> (Projects:Array of BP Animation Browser Project Object References)) -> (Output_Get:Array of BP Animation Browser Project Object References)
   > n10 = For Each Loop(Array:Array of BP Animation Browser Project Object References=@n9) -> (Array Element:BP Animation Browser Project Object Reference)
   > Add(TargetArray=Get ProjectNames(), NewItem=Get Name(self=@n10)) {is=title K2Node_CallArrayFunction}
   > SetSelectedIndex!(execute, self:"Searchable Combo Box (String) Object Reference"=Get Project_ComboBox(), Index:int=0) -> (then)
   > UpdateComboBox'() {is=title K2Node_CallFunction}
   Event PreConstruct()
  ]}
 FindSubSequence {nodes [
   n1 = FindSubSequence!() -> (then, Sequence:Movie Scene Sequence Object Reference) {is=title K2Node_FunctionEntry}
   > n2 = GetTracks!(execute, Sequence:Movie Scene Sequence Object Reference=@n1) -> (then, ReturnValue:Array of Movie Scene Track Object References)
   > n3 = For Each Loop(Array:Array of Movie Scene Track Object References=@n2) -> (Array Element:Movie Scene Track Object Reference)
   GetData!(execute:Exec=@n3.Completed, self:Self Object Reference, Sequence:Movie Scene Sequence Object Reference=@n1) -> (then:Exec)
   > n4 = For Each Loop(Array:Array of Movie Scene Sequence Object References=Get SubSequences() -> (SubSequences:Array of Movie Scene Sequence Object References)) -> (Array Element:Movie Scene Sequence Object Reference)
   > FindSubSequence(Sequence=@n4) {is=title K2Node_CallFunction}
   n5 = Cast To MovieSceneSubTrack!(execute=@n3.LoopBody, Object=@n3) -> (then, CastFailed, AsMovie Scene Sub Track:Movie Scene Sub Track Object Reference) {is=title K2Node_DynamicCast}
   > n6 = GetSections!(execute, Track:Movie Scene Track Object Reference=@n5) -> (then, ReturnValue:Array of Movie Scene Section Object References)
   > n7 = For Each Loop(Array:Array of Movie Scene Section Object References=@n6) -> (Array Element:Movie Scene Section Object Reference)
   > n8 = Cast To MovieSceneSubSection!(execute, Object=@n7) -> (then, CastFailed, AsMovie Scene Sub Section:Movie Scene Sub Section Object Reference) {is=title K2Node_DynamicCast}
   > Add Unique(
    TargetArray:Array of Movie Scene Sequence Object References=Get SubSequences()
    NewItem:"Movie Scene Sequence Object Reference (by ref)"=GetSequence!(self:Movie Scene Sub Section Object Reference=@n8) -> (ReturnValue:Movie Scene Sequence Object Reference)
   )
  ]}
 GetData {nodes [
   n1 = GetData!() -> (then, Sequence:Movie Scene Sequence Object Reference) {is=title K2Node_FunctionEntry}
   > n2 = GetBindings!(execute, Sequence:Movie Scene Sequence Object Reference=@n1) -> (then, ReturnValue:Array of Movie Scene Binding Proxy Structures)
   > n3 = Branch(Condition=Is Valid Index!(TargetArray:Array of Movie Scene Binding Proxy Structures=@n2, IndexToTest:int) -> (ReturnValue:bool))
   > n4 = For Each Loop(Array:Array of Movie Scene Binding Proxy Structures=@n2) -> (Array Element:Movie Scene Binding Proxy Structure)
   n5 = GetDataTable!(execute=[@n4.Completed, @n8.then], self) -> (then, Table:Data Table Object Reference)
   > n6 = Set EventsDT(EventsDT:Data Table Object Reference=@n5) -> (Output_Get:Data Table Object Reference)
   > n7 = Is Valid(InputObject=@n6)
   n8 = CreateNewDataTable(execute=@n7.Is Not Valid) {is=title K2Node_CallFunction}
   UpdateDataTable!(
    execute=@n7.Is Valid
    self
    DataTabkeRowArray:Array of Struct Events Asset Picker Structures=Get DataTableRowToAdd() -> (DataTableRowToAdd:Array of Struct Events Asset Picker Structures)
   ) -> (then)
   n9 = Set CurrentAssetBinding(execute=@n4.LoopBody, CurrentAssetBinding:Movie Scene Binding Proxy Structure=@n4) -> (Output_Get:Movie Scene Binding Proxy Structure)
   > n10 = GetObjectTemplate!(execute, InBinding:"Movie Scene Binding Proxy Structure (by ref)"=@n9) -> (then, ReturnValue:Object Reference)
   > n11 = Cast To FortPlayerMannequin!(execute, Object=@n10) -> (then, CastFailed, AsFort Player Mannequin:Fort Player Mannequin Object Reference) {is=title K2Node_DynamicCast}
   > Branch(Condition=IsChecked(self=Get FortManniquinCheckBox() -> (FortManniquinCheckBox:Editor Utility Check Box Object Reference)))
   > n12 = GetDisplayName!(
    execute:Exec
    InBinding:"Movie Scene Binding Proxy Structure (by ref)"=Get CurrentAssetBinding() -> (CurrentAssetBinding:Movie Scene Binding Proxy Structure)
   ) -> (then:Exec, ReturnValue:Text)
   n13 = GetAthenaCharacterItemDefinition!(self:Fort Player Mannequin Object Reference=@n11) -> (ReturnValue:Athena Character Item Definition Object Reference)
   > Add!(
    execute
    TargetArray:Array of Struct Events Asset Picker Structures=Get DataTableRowToAdd()
    NewItem_ProjectName_2_#1:String=Get Name(self=Get Current Project())
    NewItem_CommonName_27_#2:String="To String (Text)"(InText=@n12)
    NewItem_AssetName_62_#3:String=GetDisplayName!(Object:Object Reference=@n13) -> (ReturnValue:String)
    NewItem_IsFortPlayerManniquin_30_#4:bool=true
    NewItem_AthenaCharacterID_37_#5:Athena Character Item Definition Object Reference=@n13
    NewItem_AthenaBackBling_44_#6:Athena Backpack Item Definition Object Reference=Get AthenaBackBling!(self:Fort Player Mannequin Object Reference=@n11) -> (AthenaBackBling:Athena Backpack Item Definition Object Reference)
    NewItem_ExtraPart_45_#7:Athena Character Part Item Definition Object Reference=Get ExtraPart!(self:Fort Player Mannequin Object Reference=@n11) -> (ExtraPart:Athena Character Part Item Definition Object Reference)
    NewItem_SkeletalMesh_43_#8:Skeletal Mesh Object Reference=Get SkeletalMesh!(self:Skinned Mesh Component Object Reference=Get SkeletalMeshComponent!(self:Skeletal Mesh Actor Object Reference=@n11) -> (SkeletalMeshComponent:Skeletal Mesh Component Object Reference)) -> (SkeletalMesh:Skeletal Mesh Object Reference)
    NewItem_StaticMesh_48_#9:Static Mesh Object Reference
    NewItem_DefaultScale_51_#10:Vector=GetActorScale3D!(self:Actor Object Reference=@n11) -> (ReturnValue:Vector)
    NewItem_CustomClass_54_#11:Object Class Reference
    NewItem_BodyControlRig_69_#12:Control Rig Asset Soft Reference Structure=(BlueprintRigClass=None,ControlRigAsset=None)
    NewItem_FaceControlRig_70_#13:Control Rig Asset Soft Reference Structure=(BlueprintRigClass=None,ControlRigAsset=None)
    NewItem_FaceControlRig_59_#14:Control Rig Soft Class Reference=/Game/Characters/Player/Common/Fortnite_Base_Head/Facials/ControlRigs/Fortnite_3L_Face_CtrlRig.Fortnite_3L_Face_CtrlRig_C
    NewItem_BodyControlRig_57_#15:Control Rig Soft Class Reference=/FN1AnimInEd/Assets/Rigs/Templates/MR_Biped_FN1_v01.MR_Biped_FN1_v01_C
   ) -> (then, ReturnValue)
   n14 = Cast To SkeletalMeshActor!(execute=@n11.CastFailed, Object=@n10) -> (then, CastFailed, AsSkeletal Mesh Actor:Skeletal Mesh Actor Object Reference) {is=title K2Node_DynamicCast}
   > Branch(Condition=IsChecked(self=Get SkeletalMeshCheckBox() -> (SkeletalMeshCheckBox:Editor Utility Check Box Object Reference)))
   > n15 = GetDisplayName!(execute:Exec, InBinding:"Movie Scene Binding Proxy Structure (by ref)"=Get CurrentAssetBinding()) -> (then:Exec, ReturnValue:Text)
   > Add(
    TargetArray=Get DataTableRowToAdd()
    NewItem_ProjectName_2_#1=Get Name(self=Get Current Project())
    NewItem_CommonName_27_#2="To String (Text)"(InText=@n15)
    NewItem_AssetName_62_#3=GetDisplayName!(Object:Object Reference=@n14) -> (ReturnValue:String)
    NewItem_IsFortPlayerManniquin_30_#4=False
    NewItem_SkeletalMesh_43_#8=Get SkeletalMesh(self=Get SkeletalMeshComponent(self=@n14))
    NewItem_DefaultScale_51_#10=GetActorScale3D(self=@n14)
    NewItem_BodyControlRig_69_#12=(BlueprintRigClass=None,ControlRigAsset=None)
    NewItem_FaceControlRig_70_#13=(BlueprintRigClass=None,ControlRigAsset=None)
   )
   n16 = Cast To StaticMeshActor!(execute=@n14.CastFailed, Object=@n10) -> (then, CastFailed, AsStatic Mesh Actor:Static Mesh Actor Object Reference) {is=title K2Node_DynamicCast}
   > Branch(Condition=IsChecked(self=Get StaticMeshCheckBox() -> (StaticMeshCheckBox:Editor Utility Check Box Object Reference)))
   > n17 = GetDisplayName!(execute:Exec, InBinding:"Movie Scene Binding Proxy Structure (by ref)"=Get CurrentAssetBinding()) -> (then:Exec, ReturnValue:Text)
   > Add(
    TargetArray=Get DataTableRowToAdd()
    NewItem_ProjectName_2_#1=Get Name(self=Get Current Project())
    NewItem_CommonName_27_#2="To String (Text)"(InText=@n17)
    NewItem_AssetName_62_#3=GetDisplayName!(Object:Object Reference=@n16) -> (ReturnValue:String)
    NewItem_IsFortPlayerManniquin_30_#4=False
    NewItem_StaticMesh_48_#9=Get StaticMesh(self=Get StaticMeshComponent(self=@n16))
    NewItem_DefaultScale_51_#10=GetActorScale3D(self=@n16)
    NewItem_BodyControlRig_69_#12=(BlueprintRigClass=None,ControlRigAsset=None)
    NewItem_FaceControlRig_70_#13=(BlueprintRigClass=None,ControlRigAsset=None)
   )
   n18 = Cast To Actor!(execute=@n16.CastFailed, Object=@n10) -> (then, CastFailed, AsActor:Actor Object Reference) {is=title K2Node_DynamicCast}
   > Branch(Condition=IsChecked(self=Get CustomBPCheckBox() -> (CustomBPCheckBox:Editor Utility Check Box Object Reference)))
   > n19 = GetDisplayName!(execute:Exec, InBinding:"Movie Scene Binding Proxy Structure (by ref)"=Get CurrentAssetBinding()) -> (then:Exec, ReturnValue:Text)
   > Add(
    TargetArray=Get DataTableRowToAdd()
    NewItem_ProjectName_2_#1=Get Name(self=Get Current Project())
    NewItem_CommonName_27_#2="To String (Text)"(InText=@n19)
    NewItem_AssetName_62_#3=GetDisplayName!(Object:Object Reference=@n18) -> (ReturnValue:String)
    NewItem_IsFortPlayerManniquin_30_#4=False
    NewItem_DefaultScale_51_#10=GetActorScale3D(self=@n18)
    NewItem_CustomClass_54_#11=Get Class(Object=@n18)
    NewItem_BodyControlRig_69_#12=(BlueprintRigClass=None,ControlRigAsset=None)
    NewItem_FaceControlRig_70_#13=(BlueprintRigClass=None,ControlRigAsset=None)
   )
   SetWarning(execute=@n18.CastFailed, Text="LOCTEXT(\"No Assets Added\")")
   SetWarning(execute=@n3.else, Text="LOCTEXT(\"Make sure play head is in range. No data added\")")
  ]}
 GetDataTable {nodes [
   GetDataTable() {is=title K2Node_FunctionEntry}
   > n1 = GetAssetsByPaths!(
    execute
    self:Asset Registry Interface=GetAssetRegistry()
    PackagePaths:Array of Names=Make Array("[0]":Name=&DataTables) -> (Array:Array of Names)
    bRecursive:bool=true
    bIncludeOnlyOnDiskAssets:bool=false
   ) -> (then, OutAssetData:Array of Asset Data Structures, ReturnValue:bool)
   > n2 = For Each Loop!(Exec, Array:Array of Asset Data Structures=@n1) -> (
    LoopBody
    Array Element_PackageName:Name
    Array Element_PackagePath:Name
    Array Element_AssetName:Name
    Array Element_AssetClassPath_PathString:String
    Array Index
    Completed
   )
   n3 = "To String (Name)"(InName=@n2.Array Element_AssetName)
   n4 = Get Current Project()
   > Branch(Condition=Contains(SearchIn=@n3, Substring=Get Name(self=@n4)))
   > n5 = LoadAsset(AssetPath=Append!(A="To String (Name)"(InName=@n2), B=., C:String=@n3) -> (ReturnValue))
   > n6 = Cast To DataTable!(execute, Object=@n5) -> (then, CastFailed, AsData Table:Data Table Object Reference) {is=title K2Node_DynamicCast}
   > Return Node!(execute, Table:Data Table Object Reference=@n6)
  ]}
 GetProjects {nodes [
   GetProjects() {is=title K2Node_FunctionEntry}
   > Clear(TargetArray:Array of BP Animation Browser Project Object References=Get Local Projects() -> (Local Projects:Array of BP Animation Browser Project Object References))
   > n1 = GetAssetsByClass!(
    execute
    self:Asset Registry Interface=GetAssetRegistry()
    ClassPathName:Top Level Asset Path Structure=GetSoftClassTopLevelAssetPath!(SoftClassReference:Object Soft Class Reference=/AnimInEdTools/AnimationBrowser/ProjectSettings/BP_AnimationBrowser_Project.BP_AnimationBrowser_Project_C) -> (ReturnValue:Top Level Asset Path Structure)
    bSearchSubClasses:bool=false
   ) -> (then, OutAssetData:Array of Asset Data Structures, ReturnValue:bool)
   > n2 = For Each Loop!(Exec, Array=@n1) -> (LoopBody, Array Element:Asset Data Structure, Array Index, Completed)
   Return Node!(execute=@n2.Completed, Projects:Array of BP Animation Browser Project Object References=Get Local Projects())
   n3 = Break Asset Data(AssetData=@n2)
   > n4 = LoadAsset(AssetPath=GetBaseFilename(InPath="To String (Name)"(InName=@n3), bRemovePath=false))
   > n5 = Cast To BP_AnimationBrowser_Project!(execute, Object=@n4) -> (then, CastFailed, AsBP Animation Browser Project:BP Animation Browser Project Object Reference) {is=title K2Node_DynamicCast}
   > Add!(
    execute
    TargetArray:Array of BP Animation Browser Project Object References=Get Local Projects()
    NewItem:"BP Animation Browser Project Object Reference (by ref)"=@n5
   ) -> (then, ReturnValue)
  ]}
 SetItemAddedNotification {nodes [
   SetItemAddedNotification() {is=title K2Node_FunctionEntry}
   > Increment Int(Value=Get ItemsAdded() -> (ItemsAdded:int=0))
   n1 = Get WarningText() -> (WarningText:Text Object Reference)
   > SetVisibility(self=@n1, InVisibility=Visible)
   > Set!(
    execute
    ColorAndOpacity_SpecifiedColor:Linear Color Structure=(R=0.0012,G=0.1823,B=0,A=1)
    ColorAndOpacity_ColorUseRule:ESlateColorStylingMode Enum
    self:Text Object Reference=@n1
   ) -> (then, Output_Get:Slate Color Structure)
   n2 = "To String (Integer)"(InInt=Get ItemsAdded() -> (ItemsAdded=0))
   > n3 = Set Text(
    Text=To Text!(Value:String=Append!(A=@n2, B=" Item(s) added to list") -> (ReturnValue)) -> (ReturnValue:Text) {is=title K2Node_GenericToText}
    self=@n1
   )
   > PrintString(InString="To String (Text)"(InText=@n3))
  ]}
 SetItemUpdatedNotification {nodes [
   SetItemUpdatedNotification() {is=title K2Node_FunctionEntry}
   > Increment Int(Value=Get ItemsUpdated() -> (ItemsUpdated:int=0))
   n1 = Get WarningText()
   > SetVisibility(self=@n1, InVisibility=Visible)
   > Set(ColorAndOpacity_SpecifiedColor=(R=0.0012,G=0.1823,B=0,A=1), self=@n1)
   n2 = "To String (Integer)"(InInt=Get ItemsUpdated() -> (ItemsUpdated=0))
   > n3 = Set Text(Text=To Text(Value=Append(A=@n2, B=" Item(s) have been updated")), self=@n1)
   > PrintString(InString="To String (Text)"(InText=@n3))
  ]}
 SetWarning {nodes [
   SetWarning!() -> (then, Text:Text) {is=title K2Node_FunctionEntry}
   n1 = Get WarningText()
   > SetVisibility(self=@n1, InVisibility=Visible)
   > Set(ColorAndOpacity_SpecifiedColor=(R=1,G=0,B=0,A=1), self=@n1)
   > n2 = Set Text(Text=Get Text!() -> (Text), self=@n1)
   > PrintString(InString="To String (Text)"(InText=@n2))
  ]}
 UpdateComboBox {nodes [
   n1 = UpdateComboBox!() -> (then, Combo Box:"Searchable Combo Box (String) Object Reference", Items:Array of Strings, DefaultItem:String, AddBlank:bool=~) {is=title K2Node_FunctionEntry}
   > n2 = Set Current Index(Current Index:int=GetSelectedIndex!(self:"Searchable Combo Box (String) Object Reference"=@n1) -> (ReturnValue:int=0)) -> (Output_Get:int=0)
   > Set Selected Item(Selected Item:String=GetOptionAtIndex!(
     self:"Searchable Combo Box (String) Object Reference"=Get Combo Box() -> (Combo Box:"Searchable Combo Box (String) Object Reference")
     Index:int=@n2
    ) -> (ReturnValue:String)) -> (Output_Get:String)
   n3 = Get Combo Box()
   > n4 = For Loop(LastIndex=int - int(A=GetOptionCount!(self:"Searchable Combo Box (String) Object Reference"=@n3) -> (ReturnValue:int=0)))
   n5 = Branch(
    execute=@n4.Completed
    Condition=Identical!(ArrayA:Array of Strings=Get Items() -> (Items:Array of Strings), ArrayB:Array of Strings=Get Old Items() -> (Old Items:Array of Strings)) -> (ReturnValue:bool) {is=title K2Node_CallArrayFunction}
   )
   ClearOptions(execute=@n5.else, self:"Searchable Combo Box (String) Object Reference"=Get Combo Box())
   > n6 = Sequence!(execute) -> (then_0, then_1, then_2:Exec, then_3:Exec, then_4:Exec)
   > Branch(Condition=Get AddBlank() -> (AddBlank:bool))
   > AddOption!(execute, self:"Searchable Combo Box (String) Object Reference"=Get Combo Box(), Option:String) -> (then)
   n7 = Get DefaultItem() -> (DefaultItem:String)
   n8 = Branch(execute=@n6.then_1, Condition=IsEmpty(InString=@n7))
   AddOption(execute=@n8.else, self=Get Combo Box(), Option=@n7)
   n9 = For Each Loop(Exec=@n6.then_2, Array:Array of Strings=Get Items()) -> (Array Element:String)
   > AddOption(self=Get Combo Box(), Option=@n9)
   n10 = Get Combo Box()
   n11 = Get Current Index() -> (Current Index:int=0)
   n12 = Branch(
    execute=@n6.then_3
    Condition="Equal Exactly (String)"(A=GetOptionAtIndex(self=@n10, Index=@n11), B=Get Selected Item() -> (Selected Item:String))
   )
   n13 = Get Combo Box()
   > AddOption(self=@n13, Option=Temp)
   n14 = Get Current Index() -> (Current Index=0)
   > SetSelectedIndex(self=@n13, Index=int + int(A=@n14))
   > SetSelectedIndex(self=@n13, Index=@n14)
   > RemoveOption!(execute, self:"Searchable Combo Box (String) Object Reference"=@n13, Option:String=Temp) -> (then, ReturnValue:bool)
   ClearSelection(execute=@n12.else, self:"Searchable Combo Box (String) Object Reference"=Get Combo Box())
   SetSelectedOption!(execute=@n6.then_4, self:"Searchable Combo Box (String) Object Reference"=Get Combo Box(), Option:String) -> (then)
   n15 = GetOptionAtIndex(self=Get Combo Box(), Index=@n4)
   n16 = Branch(execute=@n4.LoopBody, Condition="Equal Exactly (String)"(A=@n15, B=Get DefaultItem()))
   Add(execute=@n16.else, TargetArray:Array of Strings=Get Old Items(), NewItem:"String (by ref)"=@n15)
  ]}
 UpdateDataTable {nodes [
   n1 = UpdateDataTable!() -> (then, DataTabkeRowArray:Array of Struct Events Asset Picker Structures) {is=title K2Node_FunctionEntry}
   > n2 = For Each Loop(Array:Array of Struct Events Asset Picker Structures=@n1) -> (Array Element:Struct Events Asset Picker Structure)
   > Set CurrentRow(CurrentRow:Struct Events Asset Picker Structure=@n2) -> (Output_Get:Struct Events Asset Picker Structure)
   n3 = Break Struct Events Asset Picker!(Struct_EventsAssetPicker:"Struct Events Asset Picker Structure (by ref)"=@n2) -> (
    ProjectName_2_#1:String
    CommonName_27_#2:String
    AssetName_62_#3:String
    IsFortPlayerManniquin_30_#4:bool=~
    AthenaCharacterID_37_#5:Athena Character Item Definition Object Reference
    AthenaBackBling_44_#6:Athena Backpack Item Definition Object Reference
    ExtraPart_45_#7:Athena Character Part Item Definition Object Reference
    SkeletalMesh_43_#8:Skeletal Mesh Object Reference
    StaticMesh_48_#9:Static Mesh Object Reference
    DefaultScale_51_#10:Vector=~
    CustomClass_54_#11:Object Class Reference
    BodyControlRig_69_#12:Control Rig Asset Soft Reference Structure
    FaceControlRig_70_#13:Control Rig Asset Soft Reference Structure
   ) {is=title K2Node_BreakStruct}
   > Branch(Condition=@n3.IsFortPlayerManniquin_30_#4)
   n4 = String To Name(InString=@n3.CommonName_27_#2)
   > n5 = DoesDataTableRowExist!(execute, Table:Data Table Object Reference=Get EventsDT() -> (EventsDT:Data Table Object Reference), RowName:Name=@n4) -> (then, ReturnValue:bool)
   > n6 = Branch(Condition=@n5)
   > n7 = SetItemUpdatedNotification() {is=title K2Node_CallFunction}
   AddDataTableRow!(
    execute=[@n8.then, @n7.then]
    DataTable:Data Table Object Reference=Get EventsDT()
    RowName:"Name (by ref)"=@n4
    RowData:Struct Events Asset Picker Structure=Get CurrentRow() -> (CurrentRow:Struct Events Asset Picker Structure)
   ) -> (then)
   n8 = SetItemAddedNotification(execute=@n6.else) {is=title K2Node_CallFunction}
  ]}
]
variables [
 Current Project {type=BP_AnimationBrowser_Project_C, default=null}
 CurrentProjects {type="array<BP_AnimationBrowser_Project_C>", default []}
 CurrentSequence {type=LevelSequence, default=null}
 ED_ListViewItemSelected {default=()}
 EventsDT {type=DataTable, default=/AnimInEdTools/EventsAssetLibrary/DataTables/DT_EAL_SmartTuna}
 Filter {type=string, default=""}
 FolderPaths {type="array<string>", default []}
 GetFortManni {type=boolean, default=true}
 GetSkelMesh {type=boolean, default=true}
 GetStaticMesh {type=boolean, default=true}
 ItemsAdded {type=integer}
 ItemsUpdated {type=integer}
 Level Sequence {type=LevelSequence, default=null}
 ProjectNames {type="array<string>", default []}
 Search AnimSequence Cache {type="array<string>", default []}
 Search String Split {type="array<string>", default []}
 TableRowStruct_AssetsToAdd {type="array<Struct_EventsAssetPicker>", default []}
]
widgets {
 info {rootWidgetClass=&CanvasPanel, widgetCount=27}
 widgets [
  CanvasPanel_0 {widgetClassPath=&CanvasPanel, widget=CanvasPanel_0}
  VerticalBox_0 {
   widgetClassPath=&VerticalBox
   parent=CanvasPanel_0
   widget=VerticalBox_0
   slot {
    class=CanvasPanelSlot
    properties {layoutData {alignment=<0 0>, anchors {maximum=<1 1>, minimum=<0 0>}, offsets=<ltrb 6.006 4 5.4895 5.009>}}
   }
  }
  HeaderBorder {
   widgetClassPath=&Border
   parent=VerticalBox_0
   widget=HeaderBorder
   properties {brushColor=<rgba 0.0275 0.0275 0.0275 1>, padding=<ltrb 2 2 2 2>}
   slot {class=VerticalBoxSlot, properties {padding=<ltrb 0 0 0 1>}}
  }
  VerticalBox_1 {
   widgetClassPath=&VerticalBox
   parent=HeaderBorder
   widget=VerticalBox_1
   slot {class=BorderSlot, properties {padding=<ltrb 2 2 2 2>}}
  }
  HorizontalBox_1 {widgetClassPath=&HorizontalBox, parent=VerticalBox_1, widget=HorizontalBox_1, slot {class=VerticalBoxSlot}}
  AssetImage {
   widgetClassPath=&Image
   bIsVariable=true
   parent=HorizontalBox_1
   widget=AssetImage
   properties {brush {
     bIsDynamicallyLoaded=false
     drawAs=Image
     imageType=FullColor
     mirroring=NoMirror
     resourceName=None
     resourceObject=/AnimInEdTools/AnimationBrowser/Icons/Icon_Asset
     tiling=NoTile
     imageSize=<18 28>
     margin=<ltrb 0 0 0 0>
     outlineSettings {
      bUseBrushTransparency=false
      roundingType=HalfHeightRadius
      width=0
      color {colorUseRule=UseColor_Specified, specifiedColor=<rgba 0 0 0 0>}
      cornerRadii=<0 0 0 0>
     }
     tintColor {colorUseRule=UseColor_Specified, specifiedColor=<rgba 1 1 1 1>}
     uVRegion {bIsValid=false, max=<0 0>, min=<0 0>}
    }}
   slot {class=HorizontalBoxSlot, properties {horizontalAlignment=HAlign_Center, verticalAlignment=VAlign_Center}}
  }
  SizeBox_3 {
   widgetClassPath=&SizeBox
   parent=HorizontalBox_1
   widget=SizeBox_3
   properties {
    bOverride_MaxDesiredHeight=true
    bOverride_MaxDesiredWidth=true
    bOverride_MinDesiredWidth=true
    maxDesiredHeight=25
    maxDesiredWidth=200
    minDesiredWidth=100
   }
   slot {class=HorizontalBoxSlot, properties {size {sizeRule=Fill, value=1}}}
  }
  Project_ComboBox {
   widgetClassPath=/Script/FortniteEditor.SearchableStringComboBox
   bIsVariable=true
   parent=SizeBox_3
   widget=Project_ComboBox
   properties {
    selectedOption=Project
    defaultOptions [Project]
    font {
     bForceMonospaced=false
     bMaterialIsStencil=false
     fontMaterial=null
     fontObject=&Roboto
     letterSpacing=0
     monospacedWidth=1
     size=12
     skewAmount=0
     typefaceFontName=Bold
     outlineSettings {
      bApplyOutlineToDropShadows=false
      bMiteredCorners=false
      bSeparateFillAlpha=false
      outlineMaterial=null
      outlineSize=0
      outlineColor=<rgba 0 0 0 1>
     }
    }
    foregroundColor {colorUseRule=UseColor_Specified, specifiedColor=<rgba 0.5271 0.5271 0.5271 1>}
    itemStyle {
     bUseParentRowBrush=false
     activeBrush {
      bIsDynamicallyLoaded=false
      drawAs=Image
      imageType=NoImage
      mirroring=NoMirror
      resourceName=None
      resourceObject=null
      tiling=NoTile
      imageSize=<0 0>
      margin=<ltrb 0 0 0 0>
      outlineSettings {
       bUseBrushTransparency=false
       roundingType=HalfHeightRadius
       width=0
       color {colorUseRule=UseColor_Specified, specifiedColor=<rgba 0 0 0 0>}
       cornerRadii=<0 0 0 0>
      }
      tintColor {colorUseRule=UseColor_Specified, specifiedColor=<rgba 0 0.162 0.7454 1>}
      uVRegion {bIsValid=false, max=<0 0>, min=<0 0>}
     }
     activeHighlightedBrush {
      bIsDynamicallyLoaded=false
      drawAs=Image
      imageType=NoImage
      mirroring=NoMirror
      resourceName=None
      resourceObject=null
      tiling=NoTile
      imageSize=<0 0>
      margin=<ltrb 0 0 0 0>
      outlineSettings {
       bUseBrushTransparency=false
       roundingType=HalfHeightRadius
       width=0
       color {colorUseRule=UseColor_Specified, specifiedColor=<rgba 0 0 0 0>}
       cornerRadii=<0 0 0 0>
      }
      tintColor {colorUseRule=UseColor_Specified, specifiedColor=<rgba 0.0252 0.0319 0.0423 1>}
      uVRegion {bIsValid=false, max=<0 0>, min=<0 0>}
     }
     activeHoveredBrush {
      bIsDynamicallyLoaded=false
      drawAs=Image
      imageType=NoImage
      mirroring=NoMirror
      resourceName=None
      resourceObject=null
      tiling=NoTile
      imageSize=<0 0>
      margin=<ltrb 0 0 0 0>
      outlineSettings {
       bUseBrushTransparency=false
       roundingType=HalfHeightRadius
       width=0
       color {colorUseRule=UseColor_Specified, specifiedColor=<rgba 0 0 0 0>}
       cornerRadii=<0 0 0 0>
      }
      tintColor {colorUseRule=UseColor_Specified, specifiedColor=<rgba 0 0.162 0.7454 1>}
      uVRegion {bIsValid=false, max=<0 0>, min=<0 0>}
     }
     dropIndicator_Above {
      bIsDynamicallyLoaded=false
      drawAs=Box
      imageType=FullColor
      mirroring=NoMirror
      resourceName=../../../Engine/Content/Slate/Common/DropZoneIndicator_Above.png
      resourceObject=null
      tiling=NoTile
      imageSize=<0 0>
      margin=<ltrb 0.625 0.625 0 0>
      outlineSettings {
       bUseBrushTransparency=false
       roundingType=HalfHeightRadius
       width=0
       color {colorUseRule=UseColor_Specified, specifiedColor=<rgba 0 0 0 0>}
       cornerRadii=<0 0 0 0>
      }
      tintColor {colorUseRule=UseColor_Specified, specifiedColor=<rgba 0.0091 0.3515 0.9647 1>}
      uVRegion {bIsValid=false, max=<0 0>, min=<0 0>}
     }
     dropIndicator_Below {
      bIsDynamicallyLoaded=false
      drawAs=Box
      imageType=FullColor
      mirroring=NoMirror
      resourceName=../../../Engine/Content/Slate/Common/DropZoneIndicator_Below.png
      resourceObject=null
      tiling=NoTile
      imageSize=<0 0>
      margin=<ltrb 0.625 0 0 0.625>
      outlineSettings {
       bUseBrushTransparency=false
       roundingType=HalfHeightRadius
       width=0
       color {colorUseRule=UseColor_Specified, specifiedColor=<rgba 0 0 0 0>}
       cornerRadii=<0 0 0 0>
      }
      tintColor {colorUseRule=UseColor_Specified, specifiedColor=<rgba 0.0091 0.3515 0.9647 1>}
      uVRegion {bIsValid=false, max=<0 0>, min=<0 0>}
     }
     dropIndicator_Onto {
      bIsDynamicallyLoaded=false
      drawAs=Box
      imageType=FullColor
      mirroring=NoMirror
      resourceName=../../../Engine/Content/Slate/Common/DropZoneIndicator_Onto.png
      resourceObject=null
      tiling=NoTile
      imageSize=<0 0>
      margin=<ltrb 0.25 0.25 0.25 0.25>
      outlineSettings {
       bUseBrushTransparency=false
       roundingType=HalfHeightRadius
       width=0
       color {colorUseRule=UseColor_Specified, specifiedColor=<rgba 0 0 0 0>}
       cornerRadii=<0 0 0 0>
      }
      tintColor {colorUseRule=UseColor_Specified, specifiedColor=<rgba 0.0091 0.3515 0.9647 1>}
      uVRegion {bIsValid=false, max=<0 0>, min=<0 0>}
     }
     evenRowBackgroundBrush {
      bIsDynamicallyLoaded=false
      drawAs=Image
      imageType=NoImage
      mirroring=NoMirror
      resourceName=None
      resourceObject=null
      tiling=NoTile
      imageSize=<0 0>
      margin=<ltrb 0 0 0 0>
      outlineSettings {
       bUseBrushTransparency=false
       roundingType=HalfHeightRadius
       width=0
       color {colorUseRule=UseColor_Specified, specifiedColor=<rgba 0 0 0 0>}
       cornerRadii=<0 0 0 0>
      }
      tintColor {colorUseRule=UseColor_Specified, specifiedColor=<rgba 0.0103 0.0103 0.0103 1>}
      uVRegion {bIsValid=false, max=<0 0>, min=<0 0>}
     }
     evenRowBackgroundHoveredBrush {
      bIsDynamicallyLoaded=false
      drawAs=Image
      imageType=NoImage
      mirroring=NoMirror
      resourceName=None
      resourceObject=null
      tiling=NoTile
      imageSize=<0 0>
      margin=<ltrb 0 0 0 0>
      outlineSettings {
       bUseBrushTransparency=false
       roundingType=HalfHeightRadius
       width=0
       color {colorUseRule=UseColor_Specified, specifiedColor=<rgba 0 0 0 0>}
       cornerRadii=<0 0 0 0>
      }
      tintColor {colorUseRule=UseColor_Specified, specifiedColor=<rgba 0.0176 0.0176 0.0176 1>}
      uVRegion {bIsValid=false, max=<0 0>, min=<0 0>}
     }
     inactiveBrush {
      bIsDynamicallyLoaded=false
      drawAs=Image
      imageType=NoImage
      mirroring=NoMirror
      resourceName=None
      resourceObject=null
      tiling=NoTile
      imageSize=<0 0>
      margin=<ltrb 0 0 0 0>
      outlineSettings {
       bUseBrushTransparency=false
       roundingType=HalfHeightRadius
       width=0
       color {colorUseRule=UseColor_Specified, specifiedColor=<rgba 0 0 0 0>}
       cornerRadii=<0 0 0 0>
      }
      tintColor {colorUseRule=UseColor_Specified, specifiedColor=<rgba 0.0513 0.0953 0.159 1>}
      uVRegion {bIsValid=false, max=<0 0>, min=<0 0>}
     }
     inactiveHighlightedBrush {
      bIsDynamicallyLoaded=false
      drawAs=Image
      imageType=NoImage
      mirroring=NoMirror
      resourceName=None
      resourceObject=null
      tiling=NoTile
      imageSize=<0 0>
      margin=<ltrb 0 0 0 0>
      outlineSettings {
       bUseBrushTransparency=false
       roundingType=HalfHeightRadius
       width=0
       color {colorUseRule=UseColor_Specified, specifiedColor=<rgba 0 0 0 0>}
       cornerRadii=<0 0 0 0>
      }
      tintColor {colorUseRule=UseColor_Specified, specifiedColor=<rgba 0.0252 0.0319 0.0423 1>}
      uVRegion {bIsValid=false, max=<0 0>, min=<0 0>}
     }
     inactiveHoveredBrush {
      bIsDynamicallyLoaded=false
      drawAs=Image
      imageType=NoImage
      mirroring=NoMirror
      resourceName=None
      resourceObject=null
      tiling=NoTile
      imageSize=<0 0>
      margin=<ltrb 0 0 0 0>
      outlineSettings {
       bUseBrushTransparency=false
       roundingType=HalfHeightRadius
       width=0
       color {colorUseRule=UseColor_Specified, specifiedColor=<rgba 0 0 0 0>}
       cornerRadii=<0 0 0 0>
      }
      tintColor {colorUseRule=UseColor_Specified, specifiedColor=<rgba 0.0513 0.0953 0.159 1>}
      uVRegion {bIsValid=false, max=<0 0>, min=<0 0>}
     }
     oddRowBackgroundBrush {
      bIsDynamicallyLoaded=false
      drawAs=Image
      imageType=NoImage
      mirroring=NoMirror
      resourceName=None
      resourceObject=null
      tiling=NoTile
      imageSize=<0 0>
      margin=<ltrb 0 0 0 0>
      outlineSettings {
       bUseBrushTransparency=false
       roundingType=HalfHeightRadius
       width=0
       color {colorUseRule=UseColor_Specified, specifiedColor=<rgba 0 0 0 0>}
       cornerRadii=<0 0 0 0>
      }
      tintColor {colorUseRule=UseColor_Specified, specifiedColor=<rgba 0.0103 0.0103 0.0103 1>}
      uVRegion {bIsValid=false, max=<0 0>, min=<0 0>}
     }
     oddRowBackgroundHoveredBrush {
      bIsDynamicallyLoaded=false
      drawAs=Image
      imageType=NoImage
      mirroring=NoMirror
      resourceName=None
      resourceObject=null
      tiling=NoTile
      imageSize=<0 0>
      margin=<ltrb 0 0 0 0>
      outlineSettings {
       bUseBrushTransparency=false
       roundingType=HalfHeightRadius
       width=0
       color {colorUseRule=UseColor_Specified, specifiedColor=<rgba 0 0 0 0>}
       cornerRadii=<0 0 0 0>
      }
      tintColor {colorUseRule=UseColor_Specified, specifiedColor=<rgba 0.0176 0.0176 0.0176 1>}
      uVRegion {bIsValid=false, max=<0 0>, min=<0 0>}
     }
     parentRowBackgroundBrush {
      bIsDynamicallyLoaded=false
      drawAs=Image
      imageType=NoImage
      mirroring=NoMirror
      resourceName=None
      resourceObject=null
      tiling=NoTile
      imageSize=<32 32>
      margin=<ltrb 0 0 0 0>
      outlineSettings {
       bUseBrushTransparency=false
       roundingType=HalfHeightRadius
       width=0
       color {colorUseRule=UseColor_Specified, specifiedColor=<rgba 0 0 0 0>}
       cornerRadii=<0 0 0 0>
      }
      tintColor {colorUseRule=UseColor_Specified, specifiedColor=<rgba 1 1 1 1>}
      uVRegion {bIsValid=false, max=<0 0>, min=<0 0>}
     }
     parentRowBackgroundHoveredBrush {
      bIsDynamicallyLoaded=false
      drawAs=Image
      imageType=NoImage
      mirroring=NoMirror
      resourceName=None
      resourceObject=null
      tiling=NoTile
      imageSize=<32 32>
      margin=<ltrb 0 0 0 0>
      outlineSettings {
       bUseBrushTransparency=false
       roundingType=HalfHeightRadius
       width=0
       color {colorUseRule=UseColor_Specified, specifiedColor=<rgba 0 0 0 0>}
       cornerRadii=<0 0 0 0>
      }
      tintColor {colorUseRule=UseColor_Specified, specifiedColor=<rgba 1 1 1 1>}
      uVRegion {bIsValid=false, max=<0 0>, min=<0 0>}
     }
     selectedTextColor {colorUseRule=UseColor_Specified, specifiedColor=<rgba 1 1 1 1>}
     selectorFocusedBrush {
      bIsDynamicallyLoaded=false
      drawAs=Border
      imageType=FullColor
      mirroring=NoMirror
      resourceName=../../../Engine/Content/Slate/Common/Selector.png
      resourceObject=null
      tiling=Both
      imageSize=<0 0>
      margin=<ltrb 0.25 0.25 0.25 0.25>
      outlineSettings {
       bUseBrushTransparency=false
       roundingType=HalfHeightRadius
       width=0
       color {colorUseRule=UseColor_Specified, specifiedColor=<rgba 0 0 0 0>}
       cornerRadii=<0 0 0 0>
      }
      tintColor {colorUseRule=UseColor_Specified, specifiedColor=<rgba 0.701 0.225 0.003 1>}
      uVRegion {bIsValid=false, max=<0 0>, min=<0 0>}
     }
     textColor {colorUseRule=UseColor_Specified, specifiedColor=<rgba 0.5271 0.5271 0.5271 1>}
    }
    widgetStyle {
     comboButtonStyle {
      downArrowAlign=VAlign_Center
      buttonStyle {
       clickedSlateSound {resourceObject=null}
       disabled {
        bIsDynamicallyLoaded=false
        drawAs=RoundedBox
        imageType=NoImage
        mirroring=NoMirror
        resourceName=None
        resourceObject=null
        tiling=NoTile
        imageSize=<0 0>
        margin=<ltrb 0 0 0 0>
        outlineSettings {
         bUseBrushTransparency=false
         roundingType=FixedRadius
         width=1
         color {colorUseRule=UseColor_ColorTable, specifiedColor=<rgba 0 0 0 0>}
         cornerRadii=<4 4 4 4>
        }
        tintColor {colorUseRule=UseColor_Specified, specifiedColor=<rgba 0.0395 0.0395 0.0395 1>}
        uVRegion {bIsValid=false, max=<0 0>, min=<0 0>}
       }
       disabledForeground {colorUseRule=UseColor_ColorTable, specifiedColor=<rgba 0 0 0 0>}
       hovered {
        bIsDynamicallyLoaded=false
        drawAs=RoundedBox
        imageType=NoImage
        mirroring=NoMirror
        resourceName=None
        resourceObject=null
        tiling=NoTile
        imageSize=<0 0>
        margin=<ltrb 0 0 0 0>
        outlineSettings {
         bUseBrushTransparency=false
         roundingType=FixedRadius
         width=1
         color {colorUseRule=UseColor_ColorTable, specifiedColor=<rgba 0 0 0 0>}
         cornerRadii=<4 4 4 4>
        }
        tintColor {colorUseRule=UseColor_Specified, specifiedColor=<rgba 0.0048 0.0048 0.0048 1>}
        uVRegion {bIsValid=false, max=<0 0>, min=<0 0>}
       }
       hoveredForeground {colorUseRule=UseColor_ColorTable, specifiedColor=<rgba 0 0 0 0>}
       hoveredSlateSound {resourceObject=null}
       normal {
        bIsDynamicallyLoaded=false
        drawAs=RoundedBox
        imageType=NoImage
        mirroring=NoMirror
        resourceName=None
        resourceObject=null
        tiling=NoTile
        imageSize=<0 0>
        margin=<ltrb 0 0 0 0>
        outlineSettings {
         bUseBrushTransparency=false
         roundingType=FixedRadius
         width=1
         color {colorUseRule=UseColor_ColorTable, specifiedColor=<rgba 0 0 0 0>}
         cornerRadii=<4 4 4 4>
        }
        tintColor {colorUseRule=UseColor_Specified, specifiedColor=<rgba 0.0048 0.0048 0.0048 1>}
        uVRegion {bIsValid=false, max=<0 0>, min=<0 0>}
       }
       normalForeground {colorUseRule=UseColor_ColorTable, specifiedColor=<rgba 0 0 0 0>}
       normalPadding=<ltrb 8 1 8 1>
       pressed {
        bIsDynamicallyLoaded=false
        drawAs=RoundedBox
        imageType=NoImage
        mirroring=NoMirror
        resourceName=None
        resourceObject=null
        tiling=NoTile
        imageSize=<0 0>
        margin=<ltrb 0 0 0 0>
        outlineSettings {
         bUseBrushTransparency=false
         roundingType=FixedRadius
         width=1
         color {colorUseRule=UseColor_ColorTable, specifiedColor=<rgba 0 0 0 0>}
         cornerRadii=<4 4 4 4>
        }
        tintColor {colorUseRule=UseColor_Specified, specifiedColor=<rgba 0.0048 0.0048 0.0048 1>}
        uVRegion {bIsValid=false, max=<0 0>, min=<0 0>}
       }
       pressedForeground {colorUseRule=UseColor_ColorTable, specifiedColor=<rgba 0 0 0 0>}
       pressedPadding=<ltrb 8 2 8 0>
       pressedSlateSound {resourceObject=null}
      }
      contentPadding=<ltrb 0 0 0 0>
      downArrowImage {
       bIsDynamicallyLoaded=false
       drawAs=Image
       imageType=Vector
       mirroring=NoMirror
       resourceName=../../../Engine/Content/Slate/Starship/CoreWidgets/ComboBox/wide-chevron-down.svg
       resourceObject=null
       tiling=NoTile
       imageSize=<16 16>
       margin=<ltrb 0 0 0 0>
       outlineSettings {
        bUseBrushTransparency=false
        roundingType=HalfHeightRadius
        width=0
        color {colorUseRule=UseColor_Specified, specifiedColor=<rgba 0 0 0 0>}
        cornerRadii=<0 0 0 0>
       }
       tintColor {colorUseRule=UseColor_Specified, specifiedColor=<rgba 1 1 1 1>}
       uVRegion {bIsValid=false, max=<0 0>, min=<0 0>}
      }
      downArrowPadding=<ltrb 2 2 2 2>
      menuBorderBrush {
       bIsDynamicallyLoaded=false
       drawAs=RoundedBox
       imageType=NoImage
       mirroring=NoMirror
       resourceName=None
       resourceObject=null
       tiling=NoTile
       imageSize=<0 0>
       margin=<ltrb 0 0 0 0>
       outlineSettings {
        bUseBrushTransparency=false
        roundingType=FixedRadius
        width=1
        color {colorUseRule=UseColor_ColorTable, specifiedColor=<rgba 0 0 0 0>}
        cornerRadii=<0 0 0 0>
       }
       tintColor {colorUseRule=UseColor_Specified, specifiedColor=<rgba 0.0395 0.0395 0.0395 1>}
       uVRegion {bIsValid=false, max=<0 0>, min=<0 0>}
      }
      menuBorderPadding=<ltrb 0 0 0 0>
      shadowColorAndOpacity=<rgba 0 0 0 1>
      shadowOffset=<0 0>
     }
     contentPadding=<ltrb 0 0 0 0>
     menuRowPadding=<ltrb 8 3 8 3>
     pressedSlateSound {resourceObject=null}
     selectionChangeSlateSound {resourceObject=null}
    }
   }
   slot {class=SizeBoxSlot}
  }
  HorizontalBox {
   widgetClassPath=&HorizontalBox
   parent=VerticalBox_1
   widget=HorizontalBox
   slot {class=VerticalBoxSlot, properties {padding=<ltrb 0 3 0 0>, size {sizeRule=Fill, value=1}}}
  }
  Spacer_103 {
   widgetClassPath=&Spacer
   parent=HorizontalBox
   widget=Spacer_103
   slot {class=HorizontalBoxSlot, properties {size {sizeRule=Fill, value=10}}}
  }
  FortManniquinCheckBox {
   widgetClassPath=&EditorUtilityCheckBox
   bIsVariable=true
   parent=HorizontalBox
   widget=FortManniquinCheckBox
   properties {checkedState=Checked}
   slot {class=HorizontalBoxSlot}
  }
  TextBlock_63 {
   widgetClassPath=&TextBlock
   parent=FortManniquinCheckBox
   widget=TextBlock_63
   properties {text=Fort Mannequin}
   slot {class=PanelSlot}
  }
  Spacer_161 {
   widgetClassPath=&Spacer
   parent=HorizontalBox
   widget=Spacer_161
   slot {class=HorizontalBoxSlot, properties {padding=<ltrb 5 0 5 0>}}
  }
  SkeletalMeshCheckBox {
   widgetClassPath=&EditorUtilityCheckBox
   bIsVariable=true
   parent=HorizontalBox
   widget=SkeletalMeshCheckBox
   properties {checkedState=Checked}
   slot {class=HorizontalBoxSlot}
  }
  TextBlock_115 {
   widgetClassPath=&TextBlock
   parent=SkeletalMeshCheckBox
   widget=TextBlock_115
   properties {text=Skeletal Mesh}
   slot {class=PanelSlot}
  }
  Spacer_1 {
   widgetClassPath=&Spacer
   parent=HorizontalBox
   widget=Spacer_1
   slot {class=HorizontalBoxSlot, properties {padding=<ltrb 5 0 5 0>}}
  }
  StaticMeshCheckBox {
   widgetClassPath=&EditorUtilityCheckBox
   bIsVariable=true
   parent=HorizontalBox
   widget=StaticMeshCheckBox
   properties {checkedState=Checked}
   slot {class=HorizontalBoxSlot}
  }
  TextBlock_195 {
   widgetClassPath=&TextBlock
   parent=StaticMeshCheckBox
   widget=TextBlock_195
   properties {text=Static Mesh}
   slot {class=PanelSlot}
  }
  Spacer_2 {
   widgetClassPath=&Spacer
   parent=HorizontalBox
   widget=Spacer_2
   slot {class=HorizontalBoxSlot, properties {padding=<ltrb 5 0 5 0>}}
  }
  CustomBPCheckBox {
   widgetClassPath=&EditorUtilityCheckBox
   bIsVariable=true
   parent=HorizontalBox
   widget=CustomBPCheckBox
   slot {class=HorizontalBoxSlot}
  }
  TextBlock {widgetClassPath=&TextBlock, parent=CustomBPCheckBox, widget=TextBlock, properties {text=Custom BP}, slot {class=PanelSlot}}
  Spacer {
   widgetClassPath=&Spacer
   parent=HorizontalBox
   widget=Spacer
   slot {class=HorizontalBoxSlot, properties {size {sizeRule=Fill, value=10}}}
  }
  HorizontalBox_321 {widgetClassPath=&HorizontalBox, parent=VerticalBox_0, widget=HorizontalBox_321, slot {class=VerticalBoxSlot}}
  AddAssetsBtn {
   widgetClassPath=&EditorUtilityButton
   parent=HorizontalBox_321
   widget=AddAssetsBtn
   slot {class=HorizontalBoxSlot, properties {size {sizeRule=Fill, value=1}}}
  }
  TextBlock_72 {widgetClassPath=&TextBlock, parent=AddAssetsBtn, widget=TextBlock_72, properties {text=AddAssets}, slot {class=ButtonSlot}}
  HorizontalBox_189 {
   widgetClassPath=&HorizontalBox
   parent=VerticalBox_0
   widget=HorizontalBox_189
   slot {class=VerticalBoxSlot, properties {horizontalAlignment=HAlign_Center}}
  }
  WarningText {
   widgetClassPath=&TextBlock
   bIsVariable=true
   parent=HorizontalBox_189
   widget=WarningText
   properties {text=Notification Text, visibility=Hidden, colorAndOpacity {colorUseRule=UseColor_Specified, specifiedColor=<rgba 1 0 0 1>}}
   slot {class=HorizontalBoxSlot}
  }
 ]
}
